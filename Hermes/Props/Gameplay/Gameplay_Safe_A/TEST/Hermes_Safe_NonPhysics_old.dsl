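; Test-folder, older non-physics safe variant of a searchable chest container, handling its cosmetic setup.
; On BeginPlay, where cosmetics can run, it sets smoke-sheet custom primitive data, builds mobile-only MIDs and, if unsearched, shows the smoke sheet and optionally starts an audiovisual indicator.
; An open-VFX event spawns a Niagara burst at the mesh.
; The construction script adds a tiered-chest tag for chrome material support.
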(event EventBeginPlay
  (bind _s_chest_smokesheet (Variables|Default|GetS_Chest_SmokeSheet))
  (Networking|CanExecuteCosmeticEvents
    (:True
      (Rendering|Material|SetCustomPrimitiveDataFloat _s_chest_smokesheet 0 1.0) ; WPO_HideMesh
      (Rendering|Material|SetCustomPrimitiveDataFloat _s_chest_smokesheet 1 1.0) ; Scaled_Brightness
      (CallFunction|CreateMobileMIDS)
      (if (not (Variables|Container|GetbAlreadySearched))
        (Rendering|SetVisibility (Variables|Default|GetS_Chest_SmokeSheet) true false)
        (if (Variables|Default|GetEnable%20Audio%20Indicator)
          (CallFunction|StartAudiovisualIndicator (Variables|Default|GetChest_Ambient_Sound) "Chest" (Variables|Default|GetChest%20Audio%20Update%20Timer))
          (if (Editor|InEditororPIE)
            (if (not (Variables|Container|GetbAlreadySearched))
              (Rendering|Material|SetScalarParameterValueonMaterials (Variables|Debug|GetStaticMeshComponent) "CODE_Timestamp" 0.0))))))))

(event EventDestroyed
  (|CallParentFunction))

(event Collision|EventActorBeginOverlap
  (params (OtherActor :type Actor%20Object%20Reference)))

(event EventTick
  (params (DeltaSeconds :type Float%20%28single-precision%29)))

(event EventSpawnChestOpenVFX
  ; Chest Spawn Open VFX (General)
  (Niagara|SpawnSystematLocation "/Game/Effects/Fort_Effects/World/NewChests/NS_ChestBurst_Normal.NS_ChestBurst_Normal" (.location (Transformation|GetWorldTransform (Variables|Debug|GetStaticMeshComponent))) (.rotation (Transformation|GetWorldTransform (Variables|Debug|GetStaticMeshComponent)))))

(fn CreateMobileMIDS
  (if (TouchPlatform|IsMobileGame)
    (for (_ _array_index) (Rendering|Material|GetMaterials (Variables|Debug|GetStaticMeshComponent))
      (bind _returnvalue (CallFunction|GetorCreateMIDOnDefaultMesh _array_index))
      (Utilities|Array|SetArrayElem (Variables|Default|GetMIDs) _array_index _returnvalue true))))

(fn ConstructionScript
  ;[0.32 0.9402 1] Make sure actors that are already placed in levels get this tag so that they get the right chrome material added to them.
  (|CallParentFunction)
  (Utilities|Array|AddUnique (Variables|Actor|GetTags) "Lackluster.Material.TieredChestAthena"))
